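# Sound cue for a Season 21 battle pass Week 01 voice line spoken by The Origin, using quest attenuation.
# A single dialogue player node is the root and plays the line's Dialogue Wave, with The Origin's dialogue voice as speaker and no targets.
# The cue is assigned to the narrator VO sound class and has a raised volume multiplier.
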SoundCue /BattlePassS21/Audio/VO/Cues/Week_01/PH_ORIGIN_2022_REGS_Week01_01_01_010_Dialogue_Cue
root=DialoguePlayer_0
nodes [{
  id=DialoguePlayer_0
  properties {dialogueWaveParameter {
    dialogueWave=&/PH_ORIGIN_2022_REGS_Week01_01_01_010_Dialogue
    context {speaker=&/PH_TheOrigin_DV, targets []}
   }}
 }]
properties {
 attenuationSettings=&/PH_S21_Quest_Att
 maxDistance=2097152
 soundClassObject=/Game/Sounds/FortSoundClassesAndModes/Fort_VO_Narrator
 volumeMultiplier=2
}
